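# Unlit translucent Niagara mesh-particle material that draws glitchy, flickering runes, apparently for the Cube.
# Emissive combines curve-atlas-tinted base color, stepped screen-space grid glitches, vertex-color-masked pulsing runes and scanlined noise blocks, then applies desaturation and HSV offset.
# World position offset pushes vertices horizontally from the object center using world-aligned noise, sine-driven flicker offset by actor position, and a height-based sphere pulse.
# Dynamic parameters drive block UV, offset strength and saturation.
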Material /Game/Effects/Niagara/MasterMaterials/Glitchy/M_Glitchy_CubeRunes
graph {
 properties {
  bAutomaticallySetUsageInEditor=false
  bUsedWithNiagaraMeshParticles=true
  bUsedWithStaticMesh=false
  blendMode=BLEND_Translucent
  shadingModel=MSM_Unlit
  usedShadingModels=Unlit
 }
 nodes [
  n1 = MaterialFunctionCall!(InVector=WorldPosition().XYZ) {materialFunction=/Game/Athena/Environments/Landscape/Foliage/MF_MakePreciseVec3}
  n2 = ObjectPositionWS()
  n3 = Multiply(A=Normalize(VectorInput=Subtract(A=@n1.Result, B=@n2)), B=Constant3Vector() {constant=<rgba 1 1 0 0>})
  n4 = Add(
   A=Multiply(
    A=ScalarParameter() {defaultValue=10, group=Noise, parameterName=WPOStrength}
    B=DynamicParameter() {defaultValue=<rgba 0 1 0 1>, paramNames [Block UV, WPO, FinalSaturation, Param4]}.WPO
   )
   B=Multiply(A=ObjectRadius(), B=ScalarParameter() {defaultValue=0.025, group=Noise, parameterName=WPOStrengthFromObjectRadius}) {constB=0.05}
  )
  n5 = TextureObject() {samplerType=SAMPLERTYPE_LinearColor, texture=&T_NoisePack_LinearColor}
  n6 = Divide(A=ObjectRadius(), B=ScalarParameter() {defaultValue=0.1, group=Noise, parameterName=NoiseSize})
  n7 = Multiply(A=VectorParameter() {group=Noise, parameterName=NoiseTileMult, defaultValue=<rgba 1 1 0.1 1>}.RGB, B=@n6)
  n8 = Time()
  n9 = VectorParameter() {group=Noise, parameterName=NoisePanSpeed, defaultValue=<rgba -0.08 0.07 -0.2 0>}
  n10 = Add(A=WorldPosition().XYZ, B=Multiply(A=@n6, B=Multiply(A=@n8, B=@n9.RGB)))
  n11 = Multiply(
   A=Power(Base=MaterialFunctionCall(TextureObject=@n5, TextureSize=@n7, WorldPosition=@n10) {materialFunction=&WorldAlignedTexture}.XY Texture) {constExponent=4}
   B=ScalarParameter() {defaultValue=2, group=Noise, parameterName=NoiseMaskMult}
  ) {constB=20}
  n12 = ActorPositionWS()
  n13 = MaterialFunctionCall(Float2=Round(None=Divide(A=@n12) {constB=2000})) {materialFunction=&BreakOutFloat2Components}
  n14 = Add(A=Time(), B=Add(A=@n13.R, B=@n13.R))
  n15 = Multiply(A=ScalarParameter() {defaultValue=3, parameterName=FlickerSpeed}, B=@n14)
  n16 = Sine(None=@n15) {period=1.6}
  n17 = Sine(None=@n15) {period=1.42}
  n18 = Ceil(None=AppendVector(A=AppendVector(A=@n16, B=@n17), B=Sine(None=@n15) {period=1.51}))
  n19 = ScalarParameter() {defaultValue=0.5, group=Noise, parameterName=NoiseFrequency}
  n20 = Add(A=Sine(None=@n15) {period=0.72}, B=Multiply(B=@n19) {constA=0.75}) {constB=0.75}
  n21 = Multiply(A=Add(A=Sine(None=@n15) {period=2.1}, B=Multiply(B=@n19) {constA=0.25}), B=@n20)
  n22 = Multiply(B=@n19) {constA=0.9, constB=0.9}
  n23 = Sine(None=@n15) {period=1.25}
  n24 = Multiply(A=@n20, B=Ceil(None=Add(A=@n22, B=@n23) {constB=0.9}))
  n25 = Sine(None=@n14) {period=0.075}
  n26 = Multiply(A=Saturate(None=Multiply(A=@n21, B=@n24)), B=LinearInterpolate(Alpha=ConstantBiasScale(None=@n25)) {constA=0.7})
  n27 = Add(
   A=ScalarParameter() {group=Noise, parameterName=NoiseAdd}
   B=Multiply(A=Multiply(A=@n18, B=@n26), B=ScalarParameter() {defaultValue=2, group=Noise, parameterName=NoiseMult})
  )
  n28 = Desaturation(None=Saturate(None=Multiply(A=@n11, B=@n27)))
  n29 = ComponentMask(None=WorldPosition().XYZ) {b=true}
  n30 = ComponentMask(None=MaterialFunctionCall() {materialFunction=&ObjectPivotPoint}.Object Pivot Location) {b=true}
  n31 = ScalarParameter() {defaultValue=1, parameterName=TurnOn}
  n32 = LinearInterpolate(A=@n30, B=Add(A=@n30, B=ScalarParameter() {defaultValue=300, parameterName=PulseHeight}) {constB=1600}, Alpha=@n31) {constA=-0.25, constB=1.2}
  n33 = ScalarParameter() {defaultValue=33, parameterName=PulseRadius}
  n34 = LinearInterpolate(A=@n33, B=Multiply(A=@n33) {constB=0.1}, Alpha=@n31) {constA=500, constB=100}
  n35 = Multiply(A=Multiply(A=SphereMask(A=@n29, B=@n32, Radius=@n34) {hardnessPercent=0}, B=OneMinus(None=@n31))) {constB=2}
  NamedRerouteDeclaration(None=Multiply(A=Add(A=@n28, B=@n35), B=Multiply(A=@n3, B=@n4)) {constB=0}) {name=WPO, nodeColor=<rgba 0 1 0.9297 1>}
  n36 = VertexColor()
  n37 = Multiply(A=Time())
  n38 = Add(A=Saturate(None=Sine(None=@n37))) {constB=0.2}
  n39 = TextureSample() {samplerType=SAMPLERTYPE_LinearColor, texture=/Game/Athena/Prototype/Blueprints/Cube/Materials/DFHeightMask}
  n40 = Constant() {r=3}
  n41 = Power(Base=@n39.R, Exp=@n40) {constExponent=3}
  n42 = Constant() {r=3}
  n43 = MaterialFunctionCall(In=@n41, Contrast=@n42) {materialFunction=&CheapContrast}
  n44 = Constant() {r=10}
  n45 = Multiply(A=Multiply(A=@n36.R, B=Saturate(None=@n38)), B=Add(A=@n41, B=Multiply(A=@n43.Result, B=@n44) {constB=10}))
  n46 = Add(A=Saturate(None=Cosine(None=@n37))) {constB=0.2}
  n47 = Power(Base=@n39.G, Exp=@n40) {constExponent=3}
  n48 = MaterialFunctionCall(In=@n47, Contrast=@n42) {materialFunction=&CheapContrast}
  n49 = Multiply(A=Multiply(A=@n36.G, B=Saturate(None=@n46)), B=Add(A=@n47, B=Multiply(A=@n48.Result, B=@n44) {constB=10}))
  n50 = TextureCoordinate() {uTiling=1.5}
  n51 = Add(
   A=Multiply(A=TextureSample(UVs=@n50) {samplerType=SAMPLERTYPE_LinearGrayscale, texture=/Game/Environments/Hexmap/HexClouds/T_HexmapFog_01_Height02}.R) {constB=0.5}
   B=TextureCoordinate()
  )
  n52 = Time()
  n53 = Sine(None=Add(
    A=MaterialFunctionCall(Float2=@n51) {materialFunction=&BreakOutFloat2Components}.G
    B=Multiply(A=@n52) {constB=0.5}
   ))
  NamedRerouteDeclaration(None=Multiply(A=Add(A=@n45, B=@n49), B=Saturate(None=@n53))) {name=Runes, nodeColor=<rgba 0.0312 0 1 1>}
  n54 = TextureSample() {texture=/Game/Packages/PBW/BASE/Textures/T_PBW_Base1_01}
  n55 = VectorParameter() {defaultValue=<rgba 1 1 1 0>}
  n56 = Add(A=Multiply(A=Power(Base=@n54.R, Exp=@n55.R) {constExponent=4}, B=@n55.G) {constB=10}, B=Multiply(A=@n54.R, B=@n55.B))
  n57 = CurveAtlasRowParameter(CurveTime=@n54.R) {atlas=/Game/Athena/LinearColorCurves/VFX_CurveAtlas, curve=/Game/Athena/LinearColorCurves/Curve_Contrail_GhostHunter_01, defaultValue=33}
  n58 = Multiply(A=ScreenPosition().ViewportUV) {constB=10}
  n59 = Time()
  n60 = ScalarParameter() {defaultValue=3, parameterName=Steps}
  n61 = Divide(A=Ceil(None=Multiply(A=@n59, B=@n60)), B=@n60)
  n62 = VectorParameter() {parameterName=BlocksPanSpeed, defaultValue=<rgba 1.8 -0.5 -0.8 0.6>}
  n63 = AppendVector(A=@n62.R, B=@n62.R)
  n64 = AppendVector(A=@n62.B, B=@n62.B)
  n65 = Multiply(
   A=TextureSample(UVs=Panner(Coordinate=@n58, Time=@n61, Speed=@n63) {bFractionalPart=true}) {texture=/Game/Weapons/Prototype/ObjectMover/Textures/T_Movetool_Transition}.R
   B=TextureSample(UVs=Panner(Coordinate=@n58, Time=@n61, Speed=@n64) {bFractionalPart=true}) {texture=&T_Movetool_Transition}.G
  )
  n66 = Constant() {r=5}
  n67 = Multiply(A=Time())
  n68 = Cosine(None=Sine(None=@n67) {period=2}) {period=0.5}
  n69 = Sine(None=Cosine(None=ConstantBiasScale(None=@n67)) {period=2})
  n70 = Multiply(
   A=Add(
    A=@n65
    B=MaterialFunctionCall(In=@n65, Contrast=@n66) {materialFunction=&CheapContrast}.Result
   )
   B=Multiply(A=Ceil(None=@n68), B=Ceil(None=@n69))
  )
  NamedRerouteDeclaration(None=LinearInterpolate(A=Multiply(A=@n56, B=@n57), B=VectorParameter() {parameterName=GridColor, defaultValue=<rgba 0 1 0.4121 0>}.RGB, Alpha=Saturate(None=@n70))) {name=BaseColor, nodeColor=<rgba 1 0 0.4364 1>}
  n71 = ScreenPosition()
  n72 = MaterialFunctionCall(Float2=@n71.ViewportUV) {materialFunction=&BreakOutFloat2Components}
  n73 = ScalarParameter() {defaultValue=150, group=Block, parameterName=Line amount}
  n74 = Time()
  n75 = ScalarParameter() {defaultValue=0.2, group=Block, parameterName=Line Speed}
  n76 = Frac(None=Add(A=Multiply(A=@n72.G, B=@n73), B=Multiply(A=@n74, B=@n75)))
  n77 = Multiply(A=@n76, B=OneMinus(None=@n76))
  n78 = ScalarParameter() {defaultValue=-0.15, group=Block, parameterName=Line Size}
  n79 = LinearInterpolate(
   A=ScalarParameter() {group=Block, parameterName=Line Dark}
   B=ScalarParameter() {defaultValue=1, group=Block, parameterName=Line Light}
   Alpha=Ceil(None=Add(A=@n77, B=@n78))
  )
  n80 = Add(
   A=ScalarParameter() {defaultValue=1, group=Block, parameterName=Block UV}
   B=DynamicParameter() {defaultValue=<rgba 0 1 0 1>, paramNames [Block UV, WPO, FinalSaturation, Param4]}.Block UV
  )
  n81 = Multiply(A=Time(), B=ScalarParameter() {defaultValue=10.7749, group=Block, parameterName=Block U Speed}) {constB=0.5}
  n82 = TextureSample(UVs=Panner(Coordinate=Multiply(A=@n71.ViewportUV, B=@n80), Time=Ceil(None=@n81)) {desc="-0.33,-0.43", speedX=-0.33, speedY=-0.43}) {texture=/Game/Athena/UI/Challenges/Art/T_UI_DataChip_Noise}
  n83 = Multiply(A=@n79, B=LinearInterpolate(Alpha=@n82.R) {constA=0.01})
  n84 = VectorParameter() {group=Block, parameterName=BlockColor, defaultValue=<rgba 1 1 1 0>}
  n85 = Multiply(A=@n79, B=LinearInterpolate(Alpha=@n82.G) {constA=-0.5, constB=0.5})
  n86 = VectorParameter() {group=Block, parameterName=BlockColor2, defaultValue=<rgba 1 1 1 0>}
  NamedRerouteDeclaration(None=LinearInterpolate(A=Multiply(A=@n83, B=@n84.RGB), B=Multiply(A=@n85, B=@n86.RGB))) {name=Block, nodeColor=<rgba 0 1 0.4844 1>}
  NamedRerouteDeclaration(None=@n56) {name=BaseAlpha, nodeColor=<rgba 1 0.0323 0.4052 1>}
  n87 = ParticleColor()
  n88 = NamedRerouteUsage()
  n89 = LinearInterpolate(A=@n88, B=VectorParameter() {parameterName=Runes Color, defaultValue=<rgba 0.712 0.1042 1 0>}.RGB, Alpha=NamedRerouteUsage())
  n90 = NamedRerouteUsage()
  n91 = Desaturation(
   None=Multiply(A=@n87.RGB, B=Add(A=@n89, B=@n90))
   Fraction=DynamicParameter() {defaultValue=<rgba 0 1 0 1>, paramNames [Block UV, WPO, FinalSaturation, Param4]}.FinalSaturation
  )
  n92 = VectorParameter() {group=HSV Offset, parameterName=Reference Color, defaultValue=<rgba 1 0 1 1>}
  n93 = VectorParameter() {group=HSV Offset, parameterName=HSV Offset, defaultValue=<rgba 0 0 0 1>}
  Reroute(None=MaterialFunctionCall(Emissive=MaterialFunctionCall!(RGB=@n91, HSV Offset=@n93.RGB, Optional Reference Color=@n92.RGB) {materialFunction=/Game/Effects/Niagara/Emitters/Water/FoamEffect/MF_HSV_Offset}.Result) {materialFunction=&MF_FX_EmissiveAdjustments}.Result)
  Reroute(None=NamedRerouteUsage())
  n94 = Desaturation(None=@n88)
  n95 = ScalarParameter() {group=Block, parameterName=BlockAlpha}
  n96 = Add(A=Multiply(A=Add(A=@n94, B=@n95), B=NamedRerouteUsage()), B=ScalarParameter() {parameterName=Alpha})
  Reroute(None=Multiply(A=@n87.A, B=Saturate(None=@n96)))
  TextureCoordinate()
  TextureCoordinate()
 ]
}
